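# Physics asset for the QualityCreek cosmetic backpack mesh, defining collision and joint limits for its staff bone.
# A long, thin angled capsule covers the dyn_staff bone, while a tiny capsule on the pack bone gives the staff a body to attach to.
# A single constraint joins the two with tight swing and twist limits, restricting the staff to slight movement relative to the pack.
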PhysicsAsset /BRCosmetics/Accessories/FORT_Backpacks/Backpack_M_MED_QualityCreek/Meshes/M_MED_QualityCreek_Pack_Physics
bodies {
 dyn_staff [{
   shapeType=Capsule
   length=49.0195
   radius=3.1032
   shapeName=dyn_staff_capsule
   center=<-0.1236 -9.4976 -13.1508>
   rotation=<rot 37.1549 -90 -0.1421>
  }]
 pack [{shapeType=Capsule, length=0.505, radius=0.505, shapeName=pack_capsule, rotation=<rot 0 -90 0>}]
}
constraints [{bone1Name=dyn_staff, bone2Name=pack, swing1LimitDegrees=10, swing2LimitDegrees=5, twistLimitDegrees=5}]
